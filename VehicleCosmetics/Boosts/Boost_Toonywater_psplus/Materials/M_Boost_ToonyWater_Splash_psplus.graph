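# Translucent unlit Niagara sprite material for the water splashes of the ToonyWater vehicle boost (psplus variant).
# Panning cloud noise distorts the UVs of two splash textures, which a dynamic parameter blends between.
# The blend plus a cloud-mask mist fill, scaled by particle alpha, forms the clamped opacity term.
# Caustics scaled by ReflectionAmount and a faint dark-blue base are tinted by particle color and passed through MF_FX_EmissiveAdjustments.
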Material /VehicleCosmetics/Boosts/Boost_Toonywater_psplus/Materials/M_Boost_ToonyWater_Splash_psplus
graph {
 properties {
  bUsedWithNiagaraSprites=true
  blendMode=BLEND_Translucent
  opacityMaskClipValue=0.85
  pixelDepthOffsetMode=PDOM_Legacy
  refractionMethod=RM_PixelNormalOffset
  shadingModel=MSM_Unlit
  usedShadingModels=Unlit
 }
 nodes [
  n1 = ParticleColor()
  n2 = Time()
  n3 = Constant() {r=0.5}
  n4 = TextureSample(UVs=Panner(Coordinate=TextureCoordinate() {uTiling=0.5, vTiling=0.5}, Time=Multiply(A=@n2, B=@n3)) {speedX=0.1, speedY=0.1}) {texture=/DealershipSharedLibrary/FX/Textures/Noise/T_VCVFX_CloudsSoft_RGB}
  n5 = DynamicParameter() {defaultValue=<rgba 1 1 1 1.5>, paramNames [TextureSwitch, UVDistortAmoun, MistFiller, ReflectionAmount]}
  n6 = Add(A=Multiply(A=ComponentMask(None=@n4.RGB) {g=true, r=true}, B=@n5.UVDistortAmoun), B=TextureCoordinate())
  n7 = LinearInterpolate(
   A=TextureSample(UVs=@n6) {texture=/VehicleCosmetics/Boosts/Boost_ToonyWater/Textures/T_VCVFX_Boost_ToonyWater_WaterSplash_A}.RGB
   B=TextureSample(UVs=@n6) {texture=/VehicleCosmetics/Boosts/Boost_ToonyWater/Textures/T_VCVFX_Boost_ToonyWater_WaterSplash01_A}.RGB
   Alpha=@n5.TextureSwitch
  )
  n8 = Multiply(
   A=@n5.MistFiller
   B=TextureSample(UVs=TextureCoordinate() {uTiling=2, vTiling=2}) {texture=/DealershipSharedLibrary/FX/Textures/Gradient_Masks/T_VCVFX_FireCloud02_A}.RGB
  )
  n9 = ConstantBiasScale(None=Add(A=@n7, B=Saturate(None=@n8))) {bias=-0.01, scale=2}
  Clamp(None=Multiply(A=@n1.A, B=Saturate(None=@n9))) {maxDefault=1.5}
  n10 = Constant() {r=0.1}
  n11 = TextureCoordinate()
  n12 = TextureSample(UVs=@n11) {texture=/DealershipSharedLibrary/FX/Textures/Noise/T_VCVFX_Caustics_A}
  n13 = Multiply(A=@n5.ReflectionAmount, B=TextureSample(UVs=@n11) {texture=&T_VCVFX_Caustics_A}.RGB)
  n14 = Saturate(None=@n7)
  n15 = Constant3Vector() {constant=<rgba 0.025 0.03 0.04 0>}
  n16 = Constant() {r=0.25}
  n17 = Add(A=Multiply(A=Add(A=@n10, B=@n12.RGB), B=@n5.ReflectionAmount), B=Add(A=Multiply(A=@n13, B=@n14), B=Multiply(A=@n15, B=@n16)))
  MaterialFunctionCall(Emissive=Saturate(None=Multiply(A=@n17, B=@n1.RGB))) {materialFunction=&MF_FX_EmissiveAdjustments}
 ]
}
